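; Halloween Weeping Woods ghost projectile Blueprint that configures randomized, homing movement when it spawns.
; On BeginPlay, the authority stores a point high above the actor's location, then randomizes the movement component's initial and maximum speeds.
; It passes that point into homing data using the DrunkArtillery style, with one randomized parameter, through the Fort projectile movement interface.
; The construction script is empty.
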(event EventBeginPlay
  (bind _projectilemovementcomponent (Variables|Movement|GetProjectileMovementComponent))
  (Networking|SwitchHasAuthority
    (:Authority
      (Variables|Default|SetSpawnLocationOffset (+ (Transformation|GetActorLocation self) "0, 0,2000.000000"))
      (Variables|/Script/Engine.ProjectileMovementComponent|SetInitialSpeed (Math|Random|RandomFloatinRange 1.0 5e+01) _projectilemovementcomponent)
      (Variables|/Script/Engine.ProjectileMovementComponent|SetMaxSpeed (Math|Random|RandomFloatinRange 8e+02 1e+03) _projectilemovementcomponent)
      (Projectile|SetHomingData 0 0 (Utilities|Struct|MakeProjectileHomingData "DrunkArtillery" 7.0 15.0 (Math|Random|RandomFloatinRange 3.5 5.0) 0.0 -1 0 0 (Variables|Default|GetSpawnLocationOffset)) :!class /Script/FortniteGame.FortProjectileMovementInterface))))

(fn ConstructionScript)
