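; Halloween ghost spawner for Weeping Woods that releases a burst of ghost projectiles when a player enters its trigger box at night.
; On the authority, a drop-chance roll either triggers it once or blocks further attempts until a retry timer expires.
; A looping timer feeds a DoN counter, spawning randomly offset and rotated ghosts below the actor.
; At the configured limit it pauses, then silently destroys itself after a delay.
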(event OnComponentBeginOverlap(Box)
  (params (OverlappedComponent :type Primitive%20Component%20Object%20Reference) (OtherActor :type Actor%20Object%20Reference) (OtherComp :type Primitive%20Component%20Object%20Reference) (OtherBodyIndex :type Integer) (bFromSweep :type Boolean) (SweepResult :type Hit%20Result%20Structure%20%28by%20ref%29))
  ;[1 1 1] Triggering Logic
  (Networking|SwitchHasAuthority
    (:Authority
      (Utilities|Casting|CastToFortPlayerPawnAthena OtherActor)
      (if (Variables|Default|GetAlreadyTriggered)
        _
        (elif (Variables|Default|GetSpawnBlocked)
          _
          (else
            (switch Utilities|FlowControl|Switch|SwitchonEFortDayPhase (TimeOfDay|GetCurrentDayPhase (TimeOfDay|GetContextualTimeOfDayManager) :!class /Script/FortniteGame.FortTimeOfDayManagerInterface)
              (:Night
                (bind _dropresult (Class|MissionBlueprintFunctionLibrary|DropChance (Variables|Default|GetDrop%20Chance)))
                (if _dropresult
                  (Variables|Default|SetAlreadyTriggered true)
                  (Utilities|FlowControl|DoOnce false
                    (:Completed
                      (CallFunction|SpawnGhosts)))
                  (else
                    (Variables|Default|SetSpawnBlocked true)
                    (Utilities|Time|SetTimerbyEvent (delegate Custom|AllowRetry) (Variables|Default|GetRetrySpawnTimerDuration))))))))))))

(event Custom|AllowRetry
  ;[1 1 1] Triggering Logic
  (Variables|Default|SetSpawnBlocked false))

(event Custom|SpawnGhosts
  ;[1 1 1] Spawn Ghosts
  (bind _returnvalue (Utilities|Time|SetTimerbyEvent (delegate Custom|Now) (Variables|Default|GetTimerDuration) true))
  (Variables|Default|SetTimerHandle _returnvalue))

(event Custom|Now
  ;[1 1 1] Spawn Ghosts
  (bind _counter (Utilities|FlowControl|DoN (Variables|Default|GetMaxNumberOfGhosts))
    (:Exit
      (if (>= _counter (Variables|Default|GetMaxNumberOfGhosts))
        (Utilities|Time|PauseTimerbyHandle (Variables|Default|GetTimerHandle))
        (Utilities|Time|SetTimerbyEvent (delegate Custom|DestroyNow) 2e+01)
        (else
          (Variables|Default|SetTimerDuration (Math|Random|RandomFloatinRange 0.2 0.4))
          (Game|SpawnActorfromClass "/Game/Athena/Apollo/Environments/BuildingActors/WeepingWoods/HalloweenSpecial/Apollo_GhostProjectile.Apollo_GhostProjectile_C" (Math|Transform|MakeTransform (Math|Vector|MakeVector (+ (.x (Transformation|GetActorLocation self)) (Math|Random|RandomFloatinRange (Variables|Default|GetGhostSpawnSpreadMin) (Variables|Default|GetGhostSpawnStreadMax))) (+ (.y (Transformation|GetActorLocation self)) (Math|Random|RandomFloatinRange (Variables|Default|GetGhostSpawnSpreadMin) (Variables|Default|GetGhostSpawnStreadMax))) (+ (.z (Transformation|GetActorLocation self)) (Math|Random|RandomFloatinRange -2.5e+02 -5e+01))) (Math|Rotator|MakeRotator (Math|Random|RandomFloatinRange -12.0 12.0) (Math|Random|RandomFloatinRange -12.0 12.0) (Math|Random|RandomFloatinRange 0.0 1.8e+02))) "AlwaysSpawn" "MultiplyWithRoot" 0 0))))))

(event Custom|DestroyNow
  ;[1 1 1] Spawn Ghosts
  (Gameplay|SilentDie self false))

(fn ConstructionScript
  (Components|Box|SetBoxExtent (Variables|Default|GetBox) (Variables|Default|GetTriggerBoxSize) true))
